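# Stereo MetaSound music source for SonicSpear (the LP music asset), classed as NPC music, with volume modulated by the CB_SonicSpear control bus.
# On play, it picks a wave from a one-entry music array and starts it at a random time offset alongside an ADSR attack.
# The envelope shapes both channels, which a Multiplier input then scales.
# The FadeOut trigger starts the release, and envelope completion fires OnFinished.
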MetaSoundSource /SonicSpearGameplay/Assets/Audio/Music/MSS_MX_IP_SonicSpear_LP
inputs [
 FadeOut {type=Trigger}
 Multiplier {type=Float}
 Music {type=WaveAsset:Array}
 UE.Attenuation.Distance {type=Float}
 UE.Source.OnPlay {type=Trigger}
]
interfaces [UE.Attenuation:1.0, UE.OutputFormat.Stereo:1.0, UE.Source.OneShot:1.0, UE.Source:1.1]
outputs [
 "UE.OutputFormat.Stereo.Audio:0" {type=Audio}
 "UE.OutputFormat.Stereo.Audio:1" {type=Audio}
 UE.Source.OneShot.OnFinished {type=Trigger}
]
pages [Default {
  input_defaults {
   FadeOut=null
   Multiplier=1.4
   UE.Attenuation.Distance=0
   UE.Source.OnPlay=null
   Music [/SonicSpearGameplay/Assets/Audio/Music/wav/MX_S38_POI_IP_SonicSpear_LP]
  }
  nodes [
   TriggerDelay_0 = UE.Trigger Delay(In=^UE.Source.OnPlay)
   ArrayRandomGet_0 = Array.Random Get:WaveAsset:Array(In Array=^Music, Next=@TriggerDelay_0.Out)
   n1 = UE.RandomTime(Next=@ArrayRandomGet_0.On Next) {also {defaults {Max=8, Min=1}}}
   ADSREnvelopeADSREnvelope_0 = ADSR Envelope.ADSR Envelope:Audio(Trigger Release=^FadeOut, Trigger Attack=@n1.On Next) {also {defaults {Attack Time=4, Decay Time=10, Release Time=5, Sustain Level=0.8}}}
   WavePlayer_0 = UE.Wave Player:Stereo(
    Play=@ADSREnvelopeADSREnvelope_0.On Attack Triggered
    Start Time=@n1.Value
    Wave Asset=UE.Reroute(Value=UE.Reroute(Value=@ArrayRandomGet_0.Value).Value).Value
   ) {also {defaults {Loop="<MetaSoundBuilderResult.SUCCEEDED: 0>"}, version [1,1]}}
   Multiply_2 = UE.Multiply:Audio by Float(
    AdditionalOperands=^Multiplier
    PrimaryOperand=UE.Multiply:Audio(PrimaryOperand=@ADSREnvelopeADSREnvelope_0.Out Envelope, AdditionalOperands=@WavePlayer_0.Out Right).Out
   )
   Multiply_3 = UE.Multiply:Audio by Float(
    AdditionalOperands=^Multiplier
    PrimaryOperand=UE.Multiply:Audio(PrimaryOperand=@ADSREnvelopeADSREnvelope_0.Out Envelope, AdditionalOperands=@WavePlayer_0.Out Left).Out
   )
  ]
  links [
   {dst="@UE.Source.OneShot.OnFinished", src=ADSREnvelopeADSREnvelope_0.On Done}
   {dst="@UE.OutputFormat.Stereo.Audio:1", src=Multiply_2.Out}
   {dst="@UE.OutputFormat.Stereo.Audio:0", src=Multiply_3.Out}
  ]
 }]
properties {
 assetImportData=&self:AssetImportData
 numChannels=2
 outputFormat=Stereo
 sampleRate=48000
 soundClassObject=&Fort_Music_NPC
 virtualizationMode=PlayWhenSilent
 assetUserData [&self:AudioAssetUserData_0]
 modulationSettings {volumeRouting=Override, volumeModulationDestination {bEnableModulation=true, modulators [&CB_SonicSpear]}}
}
